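# Physics asset for the WartyBrine backpack cosmetic, giving its figure-like limbs and glasses capsule bodies for constrained secondary motion.
# Small base capsules for both shoulders and legs attach to the pack body with free swing and twist.
# From each base hangs a three-body chain, shoulder-elbow-wrist or leg-knee-foot, with narrow angular limits, widest at the upper leg.
# The glasses attach directly to the pack with very tight limits, allowing only slight wobble.
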PhysicsAsset /BRCosmetics/Accessories/FORT_Backpacks/Backpack_M_MED_WartyBrine/Meshes/M_MED_WartyBrine_Pack_Physics
bodies {
 base_leg_l [{shapeType=Capsule, length=0.505, radius=0.505, shapeName=base_leg_l_capsule, rotation=<rot 0 -90 0>}]
 base_leg_r [{shapeType=Capsule, length=0.505, radius=0.505, shapeName=base_leg_r_capsule, rotation=<rot 0 -90 0>}]
 base_shoulder_l [{shapeType=Capsule, length=0.505, radius=0.505, shapeName=base_shoulder_l_capsule, rotation=<rot 0 -90 0>}]
 base_shoulder_r [{shapeType=Capsule, length=0.505, radius=0.505, shapeName=base_shoulder_r_capsule, rotation=<rot 0 -90 0>}]
 dyn_elbow_l [{shapeType=Capsule, length=1, radius=1.5, shapeName=dyn_elbow_l_capsule, center=<2 0 0>, rotation=<rot 90 0 0>}]
 dyn_elbow_r [{shapeType=Capsule, length=2, radius=1.5, shapeName=dyn_elbow_r_capsule, center=<2 0 0>, rotation=<rot 90 0 0>}]
 dyn_foot_l [{shapeType=Capsule, length=2, radius=2, shapeName=dyn_foot_l_capsule, center=<2.4294 0 0>, rotation=<rot 90 0 0>}]
 dyn_foot_r [{shapeType=Capsule, length=2, radius=2, shapeName=dyn_foot_r_capsule, center=<2.4294 0 0>, rotation=<rot 90 0 0>}]
 dyn_knee_l [{shapeType=Capsule, length=2, radius=1.9, shapeName=dyn_knee_l_capsule, center=<2.4156 0 0>, rotation=<rot 90 0 0>}]
 dyn_knee_r [{shapeType=Capsule, length=2, radius=1.9, shapeName=dyn_knee_r_capsule, center=<2.4156 0 0>, rotation=<rot 90 0 0>}]
 dyn_leg_l [{shapeType=Capsule, length=1.7, radius=1.5, shapeName=dyn_leg_l_capsule, center=<1.638 0 0>, rotation=<rot 90 0 0>}]
 dyn_leg_r [{shapeType=Capsule, length=1.7, radius=1.5, shapeName=dyn_leg_r_capsule, center=<1.638 0 0>, rotation=<rot 90 0 0>}]
 dyn_shoulder_l [{shapeType=Capsule, length=2, radius=1, shapeName=dyn_shoulder_l_capsule, center=<2 0 0>, rotation=<rot 90 0 0>}]
 dyn_shoulder_r [{shapeType=Capsule, length=2, radius=1, shapeName=dyn_shoulder_r_capsule, center=<2 0 0>, rotation=<rot 90 0 0>}]
 dyn_wrist_l [{shapeType=Capsule, length=1.5, radius=1.5, shapeName=dyn_wrist_l_capsule, center=<2 0 0>, rotation=<rot 90 0 0>}]
 dyn_wrist_r [{shapeType=Capsule, length=2, radius=1.5, shapeName=dyn_wrist_r_capsule, center=<2 0 0>, rotation=<rot 90 0 0>}]
 glasses [{shapeType=Capsule, length=2, radius=1, shapeName=glasses_capsule, center=<2 0 0>, rotation=<rot 90 0 0>}]
 pack [{shapeType=Capsule, length=0.505, radius=0.505, shapeName=pack_capsule, rotation=<rot 0 -90 0>}]
}
constraints [
 {
  bone1Name=base_leg_l
  bone2Name=pack
  swing1LimitDegrees=45
  swing1Motion=Free
  swing2LimitDegrees=45
  swing2Motion=Free
  twistLimitDegrees=45
  twistMotion=Free
 }
 {bone1Name=dyn_leg_l, bone2Name=base_leg_l, swing1LimitDegrees=30, swing2LimitDegrees=10, twistLimitDegrees=1}
 {bone1Name=dyn_knee_l, bone2Name=dyn_leg_l, swing1LimitDegrees=8, swing2LimitDegrees=1, twistLimitDegrees=1}
 {bone1Name=dyn_foot_l, bone2Name=dyn_knee_l, swing1LimitDegrees=4, swing2LimitDegrees=4, twistLimitDegrees=1}
 {
  bone1Name=base_leg_r
  bone2Name=pack
  swing1LimitDegrees=45
  swing1Motion=Free
  swing2LimitDegrees=45
  swing2Motion=Free
  twistLimitDegrees=45
  twistMotion=Free
 }
 {bone1Name=dyn_leg_r, bone2Name=base_leg_r, swing1LimitDegrees=30, swing2LimitDegrees=10, twistLimitDegrees=1}
 {bone1Name=dyn_knee_r, bone2Name=dyn_leg_r, swing1LimitDegrees=8, swing2LimitDegrees=1, twistLimitDegrees=1}
 {bone1Name=dyn_foot_r, bone2Name=dyn_knee_r, swing1LimitDegrees=4, swing2LimitDegrees=4, twistLimitDegrees=1}
 {
  bone1Name=base_shoulder_l
  bone2Name=pack
  swing1LimitDegrees=45
  swing1Motion=Free
  swing2LimitDegrees=45
  swing2Motion=Free
  twistLimitDegrees=45
  twistMotion=Free
 }
 {bone1Name=dyn_shoulder_l, bone2Name=base_shoulder_l, swing1LimitDegrees=4, swing2LimitDegrees=6, twistLimitDegrees=1}
 {bone1Name=dyn_elbow_l, bone2Name=dyn_shoulder_l, swing1LimitDegrees=6, swing2LimitDegrees=1, twistLimitDegrees=1}
 {bone1Name=dyn_wrist_l, bone2Name=dyn_elbow_l, swing1LimitDegrees=5, swing2LimitDegrees=5, twistLimitDegrees=1}
 {
  bone1Name=base_shoulder_r
  bone2Name=pack
  swing1LimitDegrees=45
  swing1Motion=Free
  swing2LimitDegrees=45
  swing2Motion=Free
  twistLimitDegrees=45
  twistMotion=Free
 }
 {bone1Name=dyn_shoulder_r, bone2Name=base_shoulder_r, swing1LimitDegrees=4, swing2LimitDegrees=6, twistLimitDegrees=1}
 {bone1Name=dyn_elbow_r, bone2Name=dyn_shoulder_r, swing1LimitDegrees=6, swing2LimitDegrees=1, twistLimitDegrees=1}
 {bone1Name=dyn_wrist_r, bone2Name=dyn_elbow_r, swing1LimitDegrees=5, swing2LimitDegrees=5, twistLimitDegrees=1}
 {bone1Name=glasses, bone2Name=pack, swing1LimitDegrees=1, swing2LimitDegrees=2, twistLimitDegrees=1}
]
